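; BP_MooneyLoader is an actor Blueprint that streams in the Mooney map and sequence levels in response to calendar events.
; On BeginPlay it waits for playlist data and, unless the playlist is tagged Athena.Creative, watches the "LoadLevel" and "LoadSequence" calendar events.
; Each event sets a RepNotify boolean, whose OnRep calls a custom event to load MooneyMap or MooneySequenceMap by name.
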(event Custom|LoadMap
  (Game|LoadStreamLevel(byName) "MooneyMap" true))

(event EventBeginPlay
  ;[1 1 1] If the playlist loaded is not creative, then set the map and sequence
  (bind _gamestate (Class|FortAsyncActionCurrentPlaylistReady|PlaylistDataReadyAsync)
    (:OnReady
      (if (not (GameplayTags|HasTag (Class|FortGameStateAthena|GetAthenaPlaylistContextTags _gamestate) "(TagName=\"Athena.Creative\")"))
        (Utilities|FlowControl|Sequence
          (:then_0
            (MCPContext|WatchCalendarEvent "LoadLevel"
              (:OnEventActive
                (Variables|Default|SetLoadMooneyMap true))))
          (:then_1
            (MCPContext|WatchCalendarEvent "LoadSequence"
              (:OnEventActive
                (Variables|Default|SetLoadMooneySequence true)))))))))

(event Custom|LoadSequence
  (Game|LoadStreamLevel(byName) "MooneySequenceMap" true))

(event Collision|EventActorBeginOverlap
  (params (OtherActor :type Actor%20Object%20Reference)))

(event EventTick
  (params (DeltaSeconds :type Float%20%28single-precision%29)))

(fn OnRep_LoadMooneyMap
  (if (Variables|Default|GetLoadMooneyMap)
    (CallFunction|LoadMap)))

(fn OnRep_LoadMooneySequence
  (if (Variables|Default|GetLoadMooneySequence)
    (CallFunction|LoadSequence)))

(fn ConstructionScript)
